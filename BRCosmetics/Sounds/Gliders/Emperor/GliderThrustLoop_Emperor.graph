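# Sound cue for the looping thrust sound of the Emperor glider.
# A mixer layers a Hi and a Lo branch, each randomly picking one of three looping waves.
# Continuous modulators on both branches map the AnyDirection parameter to pitch and volume, with the Lo layer given a wider volume range than the Hi layer.
# The cue uses glider thrust attenuation and the cosmetic glider loop audio properties sheet.
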SoundCue /BRCosmetics/Sounds/Gliders/Emperor/GliderThrustLoop_Emperor
root=Mixer_0
nodes [
 {id=Mixer_0, children [ModulatorContinuous_0, ModulatorContinuous_1], properties {inputVolume [1,1]}}
 {
  id=ModulatorContinuous_0
  children [Random_0]
  properties {
   pitchModulationParams {minOutput=0.8, parameterName=AnyDirection}
   volumeModulationParams {default=0, maxOutput=1.2, minOutput=0.7, parameterName=AnyDirection}
  }
 }
 {
  id=ModulatorContinuous_1
  children [Random_1]
  properties {
   pitchModulationParams {minOutput=0.8, parameterName=AnyDirection}
   volumeModulationParams {default=0, maxOutput=1.2, minOutput=0.2, parameterName=AnyDirection}
  }
 }
 {id=Random_0, children [WavePlayer_0,WavePlayer_1,WavePlayer_2], properties {weights [1,1,1]}}
 {id=Random_1, children [WavePlayer_3,WavePlayer_4,WavePlayer_5], properties {weights [1,1,1]}}
 {id=WavePlayer_0, properties {bLooping=true, soundWaveAssetPtr=&/Glider_Emperor_Thrust_Loop_Hi_01}}
 {id=WavePlayer_1, properties {bLooping=true, soundWaveAssetPtr=&/Glider_Emperor_Thrust_Loop_Hi_02}}
 {id=WavePlayer_2, properties {bLooping=true, soundWaveAssetPtr=&/Glider_Emperor_Thrust_Loop_Hi_03}}
 {id=WavePlayer_3, properties {bLooping=true, soundWaveAssetPtr=&/Glider_Emperor_Thrust_Loop_Lo_01}}
 {id=WavePlayer_4, properties {bLooping=true, soundWaveAssetPtr=&/Glider_Emperor_Thrust_Loop_Lo_02}}
 {id=WavePlayer_5, properties {bLooping=true, soundWaveAssetPtr=&/Glider_Emperor_Thrust_Loop_Lo_03}}
]
properties {
 attenuationSettings=&Fort_Player_Glider_Thrust_Athena_Att
 audioPropertiesSheet=&APS_Cosmetic_Glider_Loop
 maxDistance=12500
 soundClassObject=&Fort_Gliders_Land
 volumeMultiplier=0.9
}
